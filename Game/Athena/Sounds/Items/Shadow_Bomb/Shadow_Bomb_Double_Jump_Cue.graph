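# Sound cue for the Shadow Bomb double jump in Athena, classed under gadget items.
# A random node picks one of four double-jump wave variations with equal weight.
# It plays at half volume through the GhostRocks attenuation.
# Its own concurrency override caps simultaneous instances and stops the quietest when the limit is reached.
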SoundCue /Game/Athena/Sounds/Items/Shadow_Bomb/Shadow_Bomb_Double_Jump_Cue
root=Random_0
nodes [
 {id=Random_0, children [WavePlayer_0,WavePlayer_1,WavePlayer_2,WavePlayer_3], properties {weights [1,1,1,1]}}
 {id=WavePlayer_0, properties {soundWaveAssetPtr=&/WPN_Shadow_Bomb_Double_Jump_04}}
 {id=WavePlayer_1, properties {soundWaveAssetPtr=&/WPN_Shadow_Bomb_Double_Jump_03}}
 {id=WavePlayer_2, properties {soundWaveAssetPtr=&/WPN_Shadow_Bomb_Double_Jump_02}}
 {id=WavePlayer_3, properties {soundWaveAssetPtr=&/WPN_Shadow_Bomb_Double_Jump_01}}
]
properties {
 attenuationSettings=/Game/Sounds/Attenuation/Fort_GhostRocks_Att
 bOverrideConcurrency=true
 maxDistance=4500
 soundClassObject=&Fort_Gadgets_Items
 volumeMultiplier=0.5
 concurrencyOverrides {maxCount=8, resolutionRule=StopQuietest}
}
