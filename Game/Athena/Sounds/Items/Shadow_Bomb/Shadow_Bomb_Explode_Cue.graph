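# Sound cue for the Shadow Bomb item's explosion.
# A single wave player feeds a modulator left at its default settings, which is the cue's root.
# It uses the smoke grenade attenuation asset and the ranged weapon projectile effects sound class.
# A concurrency override limits it to one instance, stopping the oldest when a new one plays.
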SoundCue /Game/Athena/Sounds/Items/Shadow_Bomb/Shadow_Bomb_Explode_Cue
root=Modulator_0
nodes [
 {id=Modulator_0, children [WavePlayer_0]}
 {id=WavePlayer_0, properties {soundWaveAssetPtr=&/WPN_Shadow_Bomb_Expl_01}}
]
properties {
 attenuationSettings=/Game/Sounds/Attenuation/Fort_Grenade_Smoke_Att
 bOverrideConcurrency=true
 maxDistance=6800
 soundClassObject=&Fort_WeaponRanged_Projectile_Effects
 volumeMultiplier=1
 concurrencyOverrides {maxCount=1, resolutionRule=StopOldest}
}
