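# Sound cue for the Shadow Bomb item's jump sound in Athena, routed to the player foley sound class.
# A random node picks one of three equally weighted ninja jump waves and feeds a modulator that slightly varies a reduced volume.
# A concurrency override allows one instance at a time and blocks new plays.
# Attenuation override is enabled, so the cue's own settings supersede the referenced player footsteps attenuation asset.
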SoundCue /Game/Athena/Sounds/Items/Shadow_Bomb/Shadow_Bomb_Jump_Cue
root=Modulator_9
nodes [
 {id=Modulator_9, children [Random_5], properties {volumeMax=0.7, volumeMin=0.6}}
 {id=Random_5, children [WavePlayer_37, WavePlayer_38, WavePlayer_39], properties {weights [1,1,1]}}
 {id=WavePlayer_37, properties {soundWaveAssetPtr=/Game/Sounds/Fort_Step_Sounds/Human/ninja_jump_02}}
 {id=WavePlayer_38, properties {soundWaveAssetPtr=/Game/Sounds/Fort_Step_Sounds/Human/ninja_jump_03}}
 {id=WavePlayer_39, properties {soundWaveAssetPtr=/Game/Sounds/Fort_Step_Sounds/Human/ninja_jump_04}}
]
properties {
 attenuationSettings=/Game/Sounds/Attenuation/Fort_Player_Footsteps_Att
 bOverrideAttenuation=true
 bOverrideConcurrency=true
 maxDistance=4000
 soundClassObject=&Fort_Player_Foley
 volumeMultiplier=1
 concurrencyOverrides {maxCount=1, resolutionRule=PreventNew}
}
